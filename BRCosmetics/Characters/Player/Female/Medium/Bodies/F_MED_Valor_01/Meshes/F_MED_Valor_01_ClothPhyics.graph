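# Cloth physics asset for the F_MED_Valor_01 female medium player body, by its name the collision volume for that outfit's simulated cloth.
# It covers only the hips and upper legs: the pelvis combines two capsules with a thin, wide box slab, and each thigh has one capsule.
# Both thighs are constrained to the pelvis with swing and twist left free, so the joints impose no angular limits.
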PhysicsAsset /BRCosmetics/Characters/Player/Female/Medium/Bodies/F_MED_Valor_01/Meshes/F_MED_Valor_01_ClothPhyics
bodies {
 pelvis [
  {shapeType=Capsule, length=10, radius=5}
  {shapeType=Capsule, length=15, radius=10, center=<-13.5326 0.234 -0.0002>, rotation=<rot 90 0 0>}
  {shapeType=Box, extentX=2.5, extentY=40, extentZ=35, center=<5.7634 -0.2067 0>}
 ]
 thigh_l [{shapeType=Capsule, length=30, radius=8, center=<-15.6515 0.0747 1.8068>, rotation=<rot -88.8627 16.0891 -16.3441>}]
 thigh_r [{shapeType=Capsule, length=30, radius=8, center=<14.5144 -0.0698 -1.2797>, rotation=<rot -84.803 3.2215 -3.4869>}]
}
constraints [
 {
  bone1Name=thigh_l
  bone2Name=pelvis
  swing1LimitDegrees=45
  swing1Motion=Free
  swing2LimitDegrees=45
  swing2Motion=Free
  twistLimitDegrees=45
  twistMotion=Free
 }
 {
  bone1Name=thigh_r
  bone2Name=pelvis
  swing1LimitDegrees=45
  swing1Motion=Free
  swing2LimitDegrees=45
  swing2Motion=Free
  twistLimitDegrees=45
  twistMotion=Free
 }
]
