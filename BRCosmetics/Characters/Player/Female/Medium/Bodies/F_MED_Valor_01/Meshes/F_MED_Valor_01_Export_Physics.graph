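# Physics asset for the F_MED_Valor_01 female player body cosmetic, covering its front and back loincloth bones plus the pelvis and thighs.
# Two-segment loincloth chains use boxes at the back, linked to the pelvis, and capsules at the front, linked to a tiny master-bone capsule.
# Loincloth joints lock twist and set narrow swing limits.
# Pelvis and thigh capsules provide body volumes, with thigh-to-pelvis constraints left free on all angular axes.
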PhysicsAsset /BRCosmetics/Characters/Player/Female/Medium/Bodies/F_MED_Valor_01/Meshes/F_MED_Valor_01_Export_Physics
bodies {
 dyn_back_loincloth_01 [{shapeType=Box, extentX=13, extentY=3, extentZ=19, center=<6.1146 0.5314 0>, rotation=<rot 0 10.0004 0>}]
 dyn_back_loincloth_02 [{shapeType=Box, extentX=12, extentY=5.5, extentZ=22.5, center=<4.9253 2.2983 -0.0001>, rotation=<rot 0 10.0011 0>}]
 dyn_front_loincloth_01 [{shapeType=Capsule, length=16.6042, radius=10.1233, shapeName=dyn_front_loincloth_01_capsule, center=<13.5584 6.55 0>}]
 dyn_front_loincloth_02 [{shapeType=Capsule, length=13.1942, radius=12.798, shapeName=dyn_front_loincloth_02_capsule, center=<13.2659 6.831 0>}]
 dyn_front_loincloth_MSTR [{shapeType=Capsule, length=0.505, radius=0.505, shapeName=dyn_front_loincloth_MSTR_capsule, rotation=<rot 0 -90 0>}]
 pelvis [{shapeType=Capsule, length=14.2404, radius=11.3174, center=<-7.6903 -1.1048 0.26>}]
 thigh_l [{shapeType=Capsule, length=39.5496, radius=10.2321, shapeName=thigh_l_capsule, center=<-17.6016 0 0>, rotation=<rot 90 0 0>}]
 thigh_r [{shapeType=Capsule, length=40.1215, radius=10.2321, shapeName=thigh_r_capsule, center=<17.3947 0 0>, rotation=<rot 90 0 0>}]
}
constraints [
 {
  bone1Name=dyn_back_loincloth_02
  bone2Name=dyn_back_loincloth_01
  swing1LimitDegrees=30
  swing2LimitDegrees=10
  twistLimitDegrees=45
  twistMotion=Locked
 }
 {bone1Name=dyn_back_loincloth_01, bone2Name=pelvis, swing1LimitDegrees=20, swing2LimitDegrees=10, twistLimitDegrees=45, twistMotion=Locked}
 {
  bone1Name=dyn_front_loincloth_01
  bone2Name=dyn_front_loincloth_MSTR
  swing1LimitDegrees=20
  swing2LimitDegrees=10
  twistLimitDegrees=45
  twistMotion=Locked
 }
 {
  bone1Name=dyn_front_loincloth_02
  bone2Name=dyn_front_loincloth_01
  swing1LimitDegrees=20
  swing2LimitDegrees=10
  twistLimitDegrees=45
  twistMotion=Locked
 }
 {
  bone1Name=thigh_l
  bone2Name=pelvis
  swing1LimitDegrees=45
  swing1Motion=Free
  swing2LimitDegrees=45
  swing2Motion=Free
  twistLimitDegrees=45
  twistMotion=Free
 }
 {
  bone1Name=thigh_r
  bone2Name=pelvis
  swing1LimitDegrees=45
  swing1Motion=Free
  swing2LimitDegrees=45
  swing2Motion=Free
  twistLimitDegrees=45
  twistMotion=Free
 }
]
